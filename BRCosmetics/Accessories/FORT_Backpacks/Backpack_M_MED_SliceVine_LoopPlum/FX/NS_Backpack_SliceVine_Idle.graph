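# Idle Niagara glow effect for the SliceVine backpack cosmetic, showing two persistent sprites at mirrored offsets.
# A single burst spawns both particles, with User.Variant blending their color from yellow to green and camera offset chosen by platform.
# Alpha fades in over system age with a front-end modifier, while time of day scales brightness and soft-glow color.
# A camera-vector dot product against local Y scales cross and long flare sizes by viewing angle.
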NiagaraSystem /BRCosmetics/Accessories/FORT_Backpacks/Backpack_M_MED_SliceVine_LoopPlum/FX/NS_Backpack_SliceVine_Idle
dependencies {
 usedDataInterfaces [&NiagaraDataInterfacePlatformSet, &NiagaraDataInterfaceCurve, &NiagaraDataInterfaceCamera]
 usedDynamicInputs []
 usedModules [&SystemState]
}
emitters {Glow {
  emitterUpdateScript [
   &EmitterState(
    Inactive Response=::NewEnumerator1
    Life Cycle Mode=::NewEnumerator1
    Loop Behavior=::NewEnumerator1
    Loop Duration Mode=::NewEnumerator1
    UseLoopDelay=false
   )
   &SpawnBurst_Instantaneous(Spawn Count=2, Use Spawn Probability=false, UseLoopCountLimit=false)
  ]
  particleSpawnScript [
   &FN_SetSpawnGroup(Number of Spawn Groups=2)
   &InitializeParticle(
    Color=&LerpLinearColors(StartColor=<rgba 1.6 1.4422 0.4167 0.5>, EndColor=<rgba 0.2417 1.6 0.6176 0.5>, LerpFactor=$User.Variant)
    Lifetime=5
    Position Offset=&MergeHelper_VectorSwitch(0=<5.839 -14.9734 5.1835>, 1=<-5.839 -14.9734 5.1835>, 2=?, 3=?, Index=&SpawnGroup(), Num=::NewEnumerator0)
    Sprite Rotation Angle=-49.1631
    Sprite Rotation Mode=::NewEnumerator0
    Sprite Size Mode=::NewEnumerator3
    Uniform Sprite Size=60
    UsePositionOffset=true
    Write Material Random=true
   )
   &CameraOffset(Camera Offset Amount=&FN_PickFloatDependOnPlatform(TargetPlatformActive=4, TargetPlatformNoActive=11))
   /SetVariables(
    Particles.CrossFlareSize:Vector2f=&Multiply_Vector2D(A=$Particles.SpriteSize, B=<1.8 1.5>)
    Particles.LongFlareGlowSize:Vector2f=&Multiply_Vector2D_ByFloat(Float=1.1983, Vector2D=$Particles.SpriteSize)
    Particles.SoftGlowCameraOffset=18
   ) {isSetParameters=true}
  ]
  particleUpdateScript [
   &ParticleState(Kill Particles When Lifetime Has Elapsed=false, Let Infinitely Lived Particles Die When Emitter Deactivates=true)
   &ScaleColor(
    Scale Alpha=&FloatFromCurve(
     FloatCurve=DataInterface() {propertyValues {Curve {keys [<at 0.4 0>,<at 1 1>]}}}
     CurveIndex=$System.Age
     Scale Curve=&FN_FrontEndFloatModifier(Gameplay=1, Front End=0.8)
    )
    Scale RGB=&VectorFromFloat(Value=&FNTimeOfDayScalar(Night Scale=0.5, Morning Scale=0.8, Evening Scale=0.75))
    ScaleA=true
    ScaleRGB=true
   )
   /SetVariables(
    Particles.CrossFlareSize:Vector2f=&Multiply_Vector2D_ByFloat(
     Float=&Multiply_Float(
      A=1
      B=&DotProduct(
       Vector A=&ReturnCameraVector(Camera Vectors=::NewEnumerator4, Coordinate Space=::Simulation, Origin Position=$Particles.Position)
       Vector B=&TransformPosition(Position=<0 1 0>, Source Space=::Local, Destination Space=::Simulation)
      )
     )
     Vector2D=$Particles.Initial.CrossFlareSize
    )
    Particles.LongFlareGlowSize:Vector2f=&Multiply_Vector2D_ByFloat(
     Float=&Multiply_Float(
      A=1
      B=&DotProduct(
       Vector A=&ReturnCameraVector(Camera Vectors=::NewEnumerator4, Coordinate Space=::Simulation, Origin Position=$Particles.Position)
       Vector B=&TransformPosition(Position=<0 1 0>, Source Space=::Local, Destination Space=::Simulation)
      )
     )
     Vector2D=$Particles.Initial.LongFlareGlowSize
    )
    Particles.SoftGlowColor:LinearColor=&MultiplyLinearColor(
     Linear Color=$Particles.Color
     Scale Factor=&MakeVector4_FromFloat(Float=&FN_Select_Float_from_Time_Of_Day(Morning=0.25, Day=0.25, Evening=0.25, Night=0.4))
    )
   ) {isSetParameters=true}
  ]
  rendererClasses [&NiagaraSpriteRendererProperties]
  renderers [{rendererClass=&NiagaraSpriteRendererProperties}]
 }}
summary {
 systemName=NS_Backpack_SliceVine_Idle
 emitters [{emitterName=Glow, rendererClasses [&NiagaraSpriteRendererProperties]}]
 userVariables [
  User.bIsFrontEnd {type=NiagaraBool}
  User.bIsFrontEndPreview {type=NiagaraBool}
  User.Variant {type=NiagaraFloat, defaultValue=0, description=""}
 ]
}
